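; Oak logo quest interactable with a decal that lights up for the local player upon interaction or objective completion.
; After a delay, clients wait for playlist data, then check mode and quest state to set interactability; the server always allows it.
; Client overlap with InteractRange gates the prompt.
; Party-assisted quests stay interactable and listen for quest updates, disabling and lighting the decal's dynamic material once complete.
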(event EventBeginPlay
  (Variables|Default|SetEnabled (Utilities|Name|NotEqual(Name) (Variables|MustConfigure|GetBackendName) "None"))
  (Networking|SwitchHasAuthority
    (:Authority
      (Variables|Default|SetCanInteract true)
      (Variables|Default|SetPlayerInCollisionBox true))
    (:Remote
      (Utilities|FlowControl|Delay 5.0)
      (bind _asfort_game_state_athena (Utilities|Casting|CastToFortGameStateAthena (Game|GetGameState)))
      (if (Variables|/Script/FortniteGame.FortGameStateAthena|GetbPlaylistDataIsLoaded _asfort_game_state_athena)
        (bind _cancompletequest (Class|AthenaQuestFunctionLibrary|AthenaCheckCreativeMode (Game|GetGameState)))
        (if (and _cancompletequest (Variables|Default|GetEnabled))
          (bind (_bisassistingobjective _bisvalidplayercontroller) (Class|AthenaQuestFunctionLibrary|AthenaPartyAssistCheck (Fortnite|GetFirstLocalFortPlayerController true) (Variables|MustConfigure|GetBackendName)))
          (Variables|Default|SetbIsQuestShared (and _bisassistingobjective _bisvalidplayercontroller))
          (bind (_questisvalid _questcompleted _objectivecompleted) (Default|AthenaCheckQuestandObjectiveCompletion (Fortnite|GetFirstLocalFortPlayerController true) (Variables|ReadOnly|GetQuestDef) (Variables|MustConfigure|GetBackendName)))
          (Variables|Default|SetCanInteract (Math|Boolean|ANDBoolean _questisvalid (not _questcompleted) (not _objectivecompleted)))
          (Utilities|FlowControl|Sequence
            (:then_0
              (if _objectivecompleted
                (Rendering|Material|SetScalarParameterValue (Variables|Default|GetLightUpMID) (Variables|ReadOnly|GetMaterialParam) (Variables|ReadOnly|GetLightUpEmmissive) :!class /Script/Engine.MaterialInstanceDynamic)))
            (:then_1
              (if (Variables|Default|GetbIsQuestShared)
                (Variables|Default|SetCanInteract true)
                (bind _asfort_player_controller_athena (Utilities|Casting|CastToFortPlayerControllerAthena (Fortnite|GetFirstLocalFortPlayerController true)))
                (Quests|BindEventtoOnQuestsUpdated (Quests|GetQuestManager _asfort_player_controller_athena "Invalid") (EventDispatchers|CreateEvent :!event Custom|ParentQuestUpdated))))))
        (else
          (Playlist|BindEventtoOnCurrentPlaylistLoaded _asfort_game_state_athena (delegate Custom|PlaylistLoaded)))))))

(event Custom|PlaylistLoaded
  (params (PlaylistName :type Name) (PlaylistContextTags :type Gameplay%20Tag%20Container%20Structure%20%28by%20ref%29))
  (bind _cancompletequest (Class|AthenaQuestFunctionLibrary|AthenaCheckCreativeMode (Game|GetGameState)))
  (if (and _cancompletequest (Variables|Default|GetEnabled))
    (bind (_bisassistingobjective _bisvalidplayercontroller) (Class|AthenaQuestFunctionLibrary|AthenaPartyAssistCheck (Fortnite|GetFirstLocalFortPlayerController true) (Variables|MustConfigure|GetBackendName)))
    (Variables|Default|SetbIsQuestShared (and _bisassistingobjective _bisvalidplayercontroller))
    (bind (_questisvalid _questcompleted _objectivecompleted) (Default|AthenaCheckQuestandObjectiveCompletion (Fortnite|GetFirstLocalFortPlayerController true) (Variables|ReadOnly|GetQuestDef) (Variables|MustConfigure|GetBackendName)))
    (Variables|Default|SetCanInteract (Math|Boolean|ANDBoolean _questisvalid (not _questcompleted) (not _objectivecompleted)))
    (Utilities|FlowControl|Sequence
      (:then_0
        (if _objectivecompleted
          (Rendering|Material|SetScalarParameterValue (Variables|Default|GetLightUpMID) (Variables|ReadOnly|GetMaterialParam) (Variables|ReadOnly|GetLightUpEmmissive) :!class /Script/Engine.MaterialInstanceDynamic)))
      (:then_1
        (if (Variables|Default|GetbIsQuestShared)
          (Variables|Default|SetCanInteract true)
          (bind _asfort_player_controller_athena (Utilities|Casting|CastToFortPlayerControllerAthena (Fortnite|GetFirstLocalFortPlayerController true)))
          (Quests|BindEventtoOnQuestsUpdated (Quests|GetQuestManager _asfort_player_controller_athena "Invalid") (EventDispatchers|CreateEvent :!event Custom|ParentQuestUpdated)))))))

(event OnComponentBeginOverlap(InteractRange)
  (params (OverlappedComponent :type Primitive%20Component%20Object%20Reference) (OtherActor :type Actor%20Object%20Reference) (OtherComp :type Primitive%20Component%20Object%20Reference) (OtherBodyIndex :type Integer) (bFromSweep :type Boolean) (SweepResult :type Hit%20Result%20Structure%20%28by%20ref%29))
  (if (Variables|Default|GetEnabled)
    (Utilities|Casting|CastToPawn OtherActor)
    (Networking|SwitchHasAuthority
      (:Remote
        (Variables|Default|SetPlayerInCollisionBox true)))))

(event OnComponentEndOverlap(InteractRange)
  (params (OverlappedComponent :type Primitive%20Component%20Object%20Reference) (OtherActor :type Actor%20Object%20Reference) (OtherComp :type Primitive%20Component%20Object%20Reference) (OtherBodyIndex :type Integer))
  (if (Variables|Default|GetEnabled)
    (Utilities|Casting|CastToPawn OtherActor)
    (Networking|SwitchHasAuthority
      (:Remote
        (Variables|Default|SetPlayerInCollisionBox false)))))

(event Interaction|EventBlueprintOnInteract
  (params (InteractingPawn :type Fort%20Pawn%20Object%20Reference) (InteractionBeingAttempted :type EInteractionBeingAttempted%20Enum :default "FirstInteraction"))
  (if (Variables|Default|GetEnabled)
    (CallFunction|PlayerInteractSuccess (Fortnite|GetFortPlayerControllerfromActor InteractingPawn))))

(event Custom|PlayerInteractSuccess
  (params (Player%20Controller :type Fort%20Player%20Controller%20Object%20Reference))
  (Networking|SwitchHasAuthority
    (:Remote
      (if (Utilities|Equal(Object) Player%20Controller (Utilities|Array|Get(acopy) (Fortnite|GetLocalFortPlayerControllers true)))
        (Variables|Default|SetCanInteract false)
        (Rendering|Material|SetScalarParameterValue (Variables|Default|GetLightUpMID) (Variables|ReadOnly|GetMaterialParam) (Variables|ReadOnly|GetLightUpEmmissive) :!class /Script/Engine.MaterialInstanceDynamic)))))

(event Custom|ParentQuestUpdated
  (Networking|SwitchHasAuthority
    (:Remote
      (bind (_questisvalid _questcompleted _objectivecompleted) (Default|AthenaCheckQuestandObjectiveCompletion (Fortnite|GetFirstLocalFortPlayerController true) (Variables|ReadOnly|GetQuestDef) (Variables|MustConfigure|GetBackendName)))
      (if (or _questcompleted _objectivecompleted)
        (Variables|Default|SetCanInteract false)
        (Rendering|Material|SetScalarParameterValue (Variables|Default|GetLightUpMID) (Variables|ReadOnly|GetMaterialParam) (Variables|ReadOnly|GetLightUpEmmissive) :!class /Script/Engine.MaterialInstanceDynamic)
        (bind _asfort_player_controller_athena (Utilities|Casting|CastToFortPlayerControllerAthena (Fortnite|GetFirstLocalFortPlayerController true)))
        (Quests|UnbindEventfromOnQuestsUpdated (Quests|GetQuestManager _asfort_player_controller_athena "Invalid") (EventDispatchers|CreateEvent :!event Custom|ParentQuestUpdated))))))

(fn BlueprintCanInteract
  (params (InteractingPawn :type Fort%20Pawn%20Object%20Reference) (InteractionBeingAttempted :type EInteractionBeingAttempted%20Enum :default "FirstInteraction") (InteractionType :type TInteractionType%20Enum :default "IT_NoInteraction"))
  (return (and (Variables|Default|GetCanInteract) (Variables|Default|GetPlayerInCollisionBox))))

(fn BlueprintGetInteractionString
  (params (InteractingPawn :type Fort%20Pawn%20Object%20Reference) (InteractionBeingAttempted :type EInteractionBeingAttempted%20Enum :default "FirstInteraction"))
  (return (Variables|ReadOnly|GetInteractString)))

(fn ConstructionScript
  (bind _returnvalue (Rendering|Components|Decal|CreateDynamicMaterialInstance (Variables|Default|GetDecalVault)))
  (bind _output_get (Variables|Default|SetLightUpMID _returnvalue))
  (Rendering|Material|SetScalarParameterValue _output_get (Variables|ReadOnly|GetMaterialParam) (Variables|ReadOnly|GetBaseEmmissive) :!class /Script/Engine.MaterialInstanceDynamic))
